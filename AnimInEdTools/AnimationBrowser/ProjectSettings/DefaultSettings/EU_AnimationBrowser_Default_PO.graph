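# Default settings editor utility for the Animation Browser tool, overriding its playblast and toolbar context menu hooks.
# LoadPlayblast gets the current Level Sequence and, if its name is non-empty, runs Python that queues Animation_Playblast subtool render and explore-output actions.
# The script hard-codes a local test output folder and passes "None" as asset names, ignoring the supplied sequence name.
# LoadToolbarContextMenu only calls its parent, and the Run event is empty.
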EditorUtilityBlueprint /AnimInEdTools/AnimationBrowser/ProjectSettings/DefaultSettings/EU_AnimationBrowser_Default_PO
graphs [
 EventGraph {nodes [Event Run()]}
 LoadPlayblast {nodes [
   LoadPlayblast() {is=title K2Node_FunctionEntry}
   > n1 = GetCurrentLevelSequence!(execute) -> (then, ReturnValue:Level Sequence Object Reference)
   n2 = GetDisplayName(Object=@n1)
   > n3 = Branch(Condition=IsEmpty(InString=@n2))
   Execute Python Script!(
    execute=@n3.else
    PythonScript="import unreal\r\nfrom subtool.utils import execute_subtool_actions, SubtoolActionQueue\r\n\r\nexecute_subtool_actions(\r\n    subtool_name=\"Animation_Playblast\",\r\n    asset_names=\"None\",\r\n    action_names=[\"Render\",\"ExploreOutput\"],\r\n    subtool_overrides={\"OutputRootDir\": r\"D:\\render\\test\",\"Description\":\"AnimBrowser_Playblast\"}\r\n    )"
    asset_name:String=@n2
   ) -> (then, ReturnValue) {inputs [asset_name]}
  ]}
 LoadToolbarContextMenu {nodes [
   LoadToolbarContextMenu() {is=title K2Node_FunctionEntry}
   > Parent: LoadToolbarContextMenu() {is=title K2Node_CallParentFunction}
  ]}
]
variables []
