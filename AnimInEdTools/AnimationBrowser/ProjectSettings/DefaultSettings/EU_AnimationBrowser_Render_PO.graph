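# Editor utility for the Animation Browser that renders playblasts of the open Level Sequence through Python subtools.
# Three functions cover 4-up review, simple playblast and playblast without camera reuse, all writing to a hard-coded test folder.
# Each takes the selected browser item's name if the sequence name contains it, otherwise the sequence name.
# That name is checked against the sg_helpers asset cache, with unknown names falling back to Outfit_PipelineTest.
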EditorUtilityBlueprint /AnimInEdTools/AnimationBrowser/ProjectSettings/DefaultSettings/EU_AnimationBrowser_Render_PO
graphs [
 EventGraph {links [], nodes []}
 Playblast_4Up {nodes [
   Playblast_4Up() {is=title K2Node_FunctionEntry}
   > n1 = GetCurrentLevelSequence!(execute) -> (then, ReturnValue:Level Sequence Object Reference)
   n2 = GetDisplayName(Object=@n1)
   > n3 = Branch(Condition=IsEmpty(InString=@n2))
   n4 = Get Current List Item!(self:Animation Browser Object Reference=Get Active Animation Browser() -> (Active Animation Browser:Animation Browser Object Reference)) -> (Current List Item:Animation Browser List Item Object Reference)
   n5 = "To String (Text)"(InText=Get Text(self:Text Object Reference=Get TextBlock!(self:Animation Browser List Item Object Reference=@n4) -> (TextBlock:Text Object Reference)))
   n6 = SelectString(A=@n5, B=@n2, bPickA=Contains(SearchIn=@n2, Substring=@n5))
   n7 = Execute Python Script!(
    execute=@n3.else
    PythonScript="from unpipe import requires_plugin\r\nrequires_plugin(\"epic-shrd-techart\")\r\nfrom epic_shrd_techart.intg.common.sg_helpers.core import get_cache\r\n\r\nsg_asset = get_cache().get_asset_by_name(asset_name)\r\nout = True if sg_asset else False"
    asset_name:String=@n6
   ) -> (then, ReturnValue, out:bool=~) {inputs [asset_name], outputs [out]}
   > n8 = Branch(Condition=@n7.out)
   > Execute Python Script!(
    execute
    PythonScript="import unreal\r\nfrom subtool.utils import execute_subtool_actions, SubtoolActionQueue\r\n\r\nexecute_subtool_actions(\r\n    subtool_name=\"Animation_4up_Review\",\r\n    asset_names=[asset_name],\r\n    action_names=[\"Render\"],\r\n    subtool_overrides={\"OutputRootDir\": r\"D:\\render\\test\",\"Description\":\"AnimBrowser_4up_Playblast\"}\r\n    )"
    asset_name=@n6
   ) -> (then, ReturnValue) {inputs [asset_name]}
   Execute Python Script!(
    execute=@n8.else
    PythonScript="import unreal\r\nfrom subtool.utils import execute_subtool_actions, SubtoolActionQueue\r\n\r\nexecute_subtool_actions(\r\n    subtool_name=\"Animation_4up_Review\",\r\n    asset_names=[\"Outfit_PipelineTest\"],\r\n    action_names=[\"Render\"],\r\n    subtool_overrides={\"OutputRootDir\": r\"D:\\render\\test\",\"Description\":\"AnimBrowser_4up_Playblast\"}\r\n    )"
   ) -> (then, ReturnValue)
  ]}
 SimplePlayblast {nodes [
   SimplePlayblast() {is=title K2Node_FunctionEntry}
   > n1 = GetCurrentLevelSequence()
   n2 = GetDisplayName(Object=@n1)
   > n3 = Branch(Condition=IsEmpty(InString=@n2))
   n4 = Get Current List Item(self=Get Active Animation Browser())
   n5 = "To String (Text)"(InText=Get Text(self=Get TextBlock(self=@n4)))
   n6 = SelectString(A=@n5, B=@n2, bPickA=Contains(SearchIn=@n2, Substring=@n5))
   n7 = Execute Python Script!(
    execute=@n3.else
    PythonScript="from unpipe import requires_plugin\r\nrequires_plugin(\"epic-shrd-techart\")\r\nfrom epic_shrd_techart.intg.common.sg_helpers.core import get_cache\r\n\r\nsg_asset = get_cache().get_asset_by_name(asset_name)\r\nout = True if sg_asset else False"
    asset_name:String=@n6
   ) -> (then, ReturnValue, out:bool=~) {inputs [asset_name], outputs [out]}
   > n8 = Branch(Condition=@n7.out)
   > Execute Python Script!(
    execute
    PythonScript="import unreal\r\nfrom subtool.utils import execute_subtool_actions, SubtoolActionQueue\r\n\r\nexecute_subtool_actions(\r\n    subtool_name=\"Animation_Playblast\",\r\n    asset_names=[asset_name],\r\n    action_names=[\"Render\"],\r\n    subtool_overrides={\"OutputRootDir\": r\"D:\\render\\test\",\"Description\":\"AnimBrowser_Playblast\"}\r\n    )"
    asset_name=@n6
   ) -> (then, ReturnValue) {inputs [asset_name]}
   Execute Python Script!(
    execute=@n8.else
    PythonScript="import unreal\r\nfrom subtool.utils import execute_subtool_actions, SubtoolActionQueue\r\n\r\nexecute_subtool_actions(\r\n    subtool_name=\"Animation_Playblast\",\r\n    asset_names=[\"Outfit_PipelineTest\"],\r\n    action_names=[\"Render\"],\r\n    subtool_overrides={\"OutputRootDir\": r\"D:\\render\\test\",\"Description\":\"AnimBrowser_Playblast\"}\r\n    )"
   ) -> (then, ReturnValue)
  ]}
 SimplePlayblast_NewCameras {nodes [
   SimplePlayblast_NewCameras() {is=title K2Node_FunctionEntry}
   > n1 = GetCurrentLevelSequence()
   n2 = GetDisplayName(Object=@n1)
   > n3 = Branch(Condition=IsEmpty(InString=@n2))
   n4 = Get Current List Item(self=Get Active Animation Browser())
   n5 = "To String (Text)"(InText=Get Text(self=Get TextBlock(self=@n4)))
   n6 = SelectString(A=@n5, B=@n2, bPickA=Contains(SearchIn=@n2, Substring=@n5))
   n7 = Execute Python Script!(
    execute=@n3.else
    PythonScript="from unpipe import requires_plugin\r\nrequires_plugin(\"epic-shrd-techart\")\r\nfrom epic_shrd_techart.intg.common.sg_helpers.core import get_cache\r\n\r\nsg_asset = get_cache().get_asset_by_name(asset_name)\r\nout = True if sg_asset else False"
    asset_name:String=@n6
   ) -> (then, ReturnValue, out:bool=~) {inputs [asset_name], outputs [out]}
   > n8 = Branch(Condition=@n7.out)
   > Execute Python Script!(
    execute
    PythonScript="import unreal\r\nfrom subtool.utils import execute_subtool_actions, SubtoolActionQueue\r\n\r\nexecute_subtool_actions(\r\n    subtool_name=\"Animation_Playblast\",\r\n    asset_names=[asset_name],\r\n    action_names=[\"Render\"],\r\n    subtool_overrides={\"OutputRootDir\": r\"D:\\render\\test\",\"Description\":\"AnimBrowser_Playblast\",\"Reuse Camera\" : False}\r\n    )"
    asset_name=@n6
   ) -> (then, ReturnValue) {inputs [asset_name]}
   Execute Python Script!(
    execute=@n8.else
    PythonScript="import unreal\r\nfrom subtool.utils import execute_subtool_actions, SubtoolActionQueue\r\n\r\nexecute_subtool_actions(\r\n    subtool_name=\"Animation_Playblast\",\r\n    asset_names=[\"Outfit_PipelineTest\"],\r\n    action_names=[\"Render\"],\r\n    subtool_overrides={\"OutputRootDir\": r\"D:\\render\\test\",\"Description\":\"AnimBrowser_Playblast\",\"Reuse Camera\" : False}\r\n    )"
   ) -> (then, ReturnValue)
  ]}
]
variables []
